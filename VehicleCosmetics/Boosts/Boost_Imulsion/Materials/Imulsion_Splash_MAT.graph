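# Translucent unlit Niagara sprite material for splash particles in the Imulsion boost vehicle cosmetic.
# Particle color blends two splash masks and their matching normal maps, with inverted particle alpha subtracted from the mask for an erosion-style dissolve.
# The transformed normal is dotted with a Z-up vector to fake directional shading on a squared HDR orange color, which feeds MF_FX_EmissiveAdjustments.
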Material /VehicleCosmetics/Boosts/Boost_Imulsion/Materials/Imulsion_Splash_MAT
graph {
 properties {bUsedWithNiagaraSprites=true, blendMode=BLEND_Translucent, pixelDepthOffsetMode=PDOM_Legacy, shadingModel=MSM_Unlit, usedShadingModels=Unlit}
 nodes [
  MaterialExpressionVertexColor_0 = ParticleColor()
  n1 = OneMinus(None=@MaterialExpressionVertexColor_0.A)
  n2 = LinearInterpolate(
   A=TextureSample() {texture=/DealershipSharedLibrary/FX/Textures/Gradient_Masks/T_VCVFX_Splash02_A}.RGB
   B=TextureSampleParameter2D() {parameterName=Texture, texture=/DealershipSharedLibrary/FX/Textures/Gradient_Masks/T_VCVFX_Splash01_A}.RGB
   Alpha=@MaterialExpressionVertexColor_0.R
  )
  MaterialExpressionConstantClamp_0 = Saturate(None=Multiply(A=Subtract(A=@n2, B=@n1), B=Constant() {r=3}))
  n3 = VectorParameter() {defaultValue=<rgba 10 3.2217 0.5729 3>}
  n4 = TextureSample() {samplerType=SAMPLERTYPE_Normal, texture=/DealershipSharedLibrary/FX/Textures/Normal/T_VCVFX_Splash02_N}
  n5 = TextureSampleParameter2D() {parameterName=Normal, samplerType=SAMPLERTYPE_Normal, texture=/DealershipSharedLibrary/FX/Textures/Normal/T_VCVFX_Splash01_N}
  n6 = DotProduct(
   A=Transform(None=LinearInterpolate(A=@n4.RGB, B=@n5.RGB, Alpha=@MaterialExpressionVertexColor_0.RGB))
   B=Constant3Vector() {constant=<rgba 0 0 1 0>}
  )
  MaterialFunctionCall(Emissive=Multiply(A=Multiply(A=@n3.RGB, B=@n3.RGB), B=ConstantBiasScale(None=@n6) {bias=1.2})) {materialFunction=&MF_FX_EmissiveAdjustments}
 ]
}
