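# Masked, unlit, two-sided particle material for the Imulsion vehicle boost, usable on sprites and mesh particles.
# A SubUV texture's RGB acts as normals, driving a tinted fresnel rim and fake directional shading from the transformed Z component.
# Its red and green channels also distort a tiled lava detail texture that blends two squared colors.
# Rim and shaded lava are summed through MF_FX_EmissiveAdjustments into emissive.
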Material /VehicleCosmetics/Boosts/Boost_Imulsion/Materials/New_Imulsion_MAT
graph {
 properties {
  bUsedWithMeshParticles=true
  bUsedWithNiagaraSprites=true
  bUsedWithParticleSprites=true
  blendMode=BLEND_Masked
  opacityMaskClipValue=0.6
  pixelDepthOffsetMode=PDOM_Legacy
  shadingModel=MSM_Unlit
  twoSided=true
  usedShadingModels=Unlit
 }
 nodes [
  n1 = TextureSampleParameterSubUV!(UVs, Apply View MipBias) {
   parameterName=Texture
   samplerType=SAMPLERTYPE_LinearColor
   texture=/VehicleCosmetics/Boosts/Boost_Imulsion/Textures/T_VCVFX_Boost_Imulsion_TEST_05
  }
  n2 = OneMinus(None=DotProduct(A=CameraVectorWS(), B=@n1.RGB))
  n3 = ScalarParameter() {defaultValue=3, parameterName=Fresnel_Exponent}
  MaterialExpressionConstantClamp_4 = Saturate(None=Multiply(A=Power(Base=@n2, Exp=@n3), B=ScalarParameter() {defaultValue=3, parameterName=Fresnel_Sharpness}))
  n4 = VectorParameter() {parameterName=FresnelColor, defaultValue=<rgba 1 0.5 0.2 1>}
  n5 = Multiply(A=@n4.RGB, B=@n4.RGB)
  n6 = VectorParameter() {parameterName=Color1, defaultValue=<rgba 0.3 0.3 0.1 2>}
  n7 = VectorParameter() {parameterName=Color, defaultValue=<rgba 2 1 0.55 8>}
  n8 = Add(A=TextureCoordinate() {uTiling=8, vTiling=8}, B=Multiply(A=ComponentMask(None=@n1.RGB) {g=true, r=true}, B=Constant() {r=0.1}))
  MaterialExpressionConstantClamp_5 = Saturate(None=Subtract(
    A=TextureSample(UVs=@n8) {texture=/VehicleCosmetics/Boosts/Boost_Imulsion/Textures/T_VCVFX_Boost_Imulsion_LavaHot_Detail}.R
    B=ScalarParameter() {defaultValue=0.03, parameterName=DirtAmount}
   ))
  n9 = LinearInterpolate(A=Multiply(A=@n6.RGB, B=@n6.RGB), B=Multiply(A=@n7.RGB, B=@n7.RGB), Alpha=@MaterialExpressionConstantClamp_5)
  MaterialExpressionConstantClamp_6 = Saturate(None=DotProduct(A=Transform(None=@n1.RGB), B=Constant3Vector() {constant=<rgba 0 0 1 0>}))
  n10 = Multiply(A=Add(A=@MaterialExpressionConstantClamp_6, B=ScalarParameter() {defaultValue=0.4, parameterName=Bias}), B=Constant() {r=0.5})
  MaterialFunctionCall(Emissive=Add(A=Multiply(A=@MaterialExpressionConstantClamp_4, B=@n5), B=Multiply(A=@n9, B=@n10))) {materialFunction=&MF_FX_EmissiveAdjustments}
  Reroute(None=@n1.A)
 ]
}
